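# UI-domain material for the Red Oasis item shop tile, tinting its tile texture in one of three color schemes.
# The texture's RGB blends between two constant colors per scheme.
# Nested static switches select the palette at compile time: Black (near-black), otherwise Green, otherwise pale white-grey.
# A separate lerp driven by a scaled-to-fit radial gradient has no color inputs and does not feed the switches.
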Material /ItemShop/UI/Art/FFC/M_UI_ShopTile_RedOasis
graph {
 properties {bUsedWithStaticMesh=false, materialDomain=MD_UI, pixelDepthOffsetMode=PDOM_Legacy, usedShadingModels=Default Lit}
 nodes [
  LinearInterpolate(Alpha=MaterialFunctionCall(
    UV=MaterialFunctionCall() {materialFunction=&MF_UI_UVScaling}.Scaled to Fit
    GradientStart=Constant() {r=-0.005}
    GradientEnd=Constant()
   ) {materialFunction=&MF_UI_RadialGradient}.Smooth)
  n1 = Constant3Vector() {constant=<rgba 0.0199 0.9531 0.0199 0>}
  n2 = Constant3Vector() {constant=<rgba 0.5578 0.9167 0.4345 1>}
  n3 = TextureSample() {texture=/OfferCatalog/Art/FFC/T_UI_RedOasisShopTile}
  n4 = Constant3Vector() {constant=<rgba 0.9624 1 0.9789 0>}
  n5 = Constant3Vector() {constant=<rgba 0.7135 0.7135 0.7135 1>}
  n6 = TextureSample() {texture=&T_UI_RedOasisShopTile}
  StaticSwitchParameter(
   True=LinearInterpolate(
    A=Constant3Vector()
    B=Constant3Vector() {constant=<rgba 0.069 0.0729 0.0681 0>}
    Alpha=TextureSample() {texture=&T_UI_RedOasisShopTile}.RGB
   )
   False=StaticSwitchParameter(True=LinearInterpolate(A=@n1, B=@n2, Alpha=@n3.RGB), False=LinearInterpolate(A=@n4, B=@n5, Alpha=@n6.RGB)) {parameterName=Green}
  ) {parameterName=Black}
 ]
}
